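; A Battle Pass Season 21 narrative quest actor for a Week 8 comm device, built as a child of a parent quest actor.
; BeginPlay and the construction script defer to the parent, while overlap and tick are disabled stubs.
; After interaction, the VFX and audio event shrinks the interaction collision scale to a tenth on each axis and prints the new scale as a debug string.
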(event EventBeginPlay
  (|CallParentFunction))

(event Collision|EventActorBeginOverlap
  (params (OtherActor :type Actor%20Object%20Reference))
  (|CallParentFunction OtherActor) ; This node is disabled and will not be called.\nDrag off pins to build functionality.
)

(event EventTick
  (params (DeltaSeconds :type Float%20%28single-precision%29))
  (|CallParentFunction DeltaSeconds) ; This node is disabled and will not be called.\nDrag off pins to build functionality.
)

(event VisualState|VFX&Audio|EventPostInteractionVFX&Audio
  (params (PlayerController :type Fort%20Player%20Controller%20Object%20Reference))
  (bind _output_get (Variables|Design|QuestSetup|InteractionSettings|SetInteraction%20Collision%20Scale "0.100000,0.100000,0.100000"))
  (Development|PrintString (Utilities|String|ToString(Vector) _output_get)))

(fn ConstructionScript
  (|CallParentFunction))
