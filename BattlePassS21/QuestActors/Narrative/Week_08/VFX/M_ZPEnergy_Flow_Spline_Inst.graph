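# Material for a ZP energy flow spline effect, judging by its name, among the Week 08 narrative quest actor VFX of BattlePassS21.
# It uses Default Lit shading and holds a single unnamed vector parameter with a dark grey default.
# No connections to material outputs, textures or flow animation logic appear here, despite the Flow and Inst naming.
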Material /BattlePassS21/QuestActors/Narrative/Week_08/VFX/M_ZPEnergy_Flow_Spline_Inst
graph {
 properties {pixelDepthOffsetMode=PDOM_Legacy, usedShadingModels=Default Lit}
 nodes [VectorParameter() {defaultValue=<rgba 0.2159 0.2159 0.2159 0>}]
}
